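# Material function that applies a Fresnel mask to emissive color, in the Twin Light Purple skin of a female player character.
# It reads Normal and EmissiveColor from incoming material attributes, transforming the normal to drive an adjustable-exponent Fresnel.
# A scalar blends between inverted and regular Fresnel, and the result multiplies the emissive, which is written back.
# A static switch, on by default, can pass the attributes through unchanged.
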MaterialFunction /BRCosmetics/Characters/Player/Female/Medium/Bodies/F_MED_Twin_Dark/Skins/Twin_Light_Purple/Materials/MF_EmissiveFresnel_twinlight
graph {nodes [
  n1 = GetMaterialAttributes(None=FunctionInput() {inputType=FunctionInput_MaterialAttributes}) {attributeGetTypes [#1, #2]}
  n2 = Saturate(None=Fresnel(
    ExponentIn=ScalarParameter() {defaultValue=1, group=Emissive, parameterName=Emissive Fres EX}
    BaseReflectFractionIn=Constant()
    Normal=Transform(None=@n1.Normal)
   ) {baseReflectFraction=0})
  n3 = OneMinus(None=@n2)
  n4 = ScalarParameter() {defaultValue=0.5, group=Emissive, parameterName=Invert Emissive Fresnel}
  n5 = Multiply(A=@n1.EmissiveColor, B=Saturate(None=LinearInterpolate(A=@n3, B=@n2, Alpha=@n4)))
  FunctionOutput(None=StaticSwitchParameter(
    True=SetMaterialAttributes!(MaterialAttributes=@n1.MaterialAttributes, EmissiveColor=@n5) {attributeSetTypes [#1]}
    False=@n1.MaterialAttributes
   ) {defaultValue=true, group=Emissive, parameterName=Use Emissive Fresnel})
 ]}
