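# Material function for the Twinlight Purple character skin, producing an animated, swirling color pattern with a red Fresnel rim.
# A flame texture, projected world-aligned on pre-skinned position and scrolled over time, indexes a purple curve atlas for color.
# Fresnel varies pattern brightness and, masked by the texture's green channel, blends in over-bright red.
# Side branches build a scrolling turbulence position offset and a seam-adjusted pre-skinned normal as named reroutes.
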MaterialFunction /BRCosmetics/Characters/Player/Female/Medium/Bodies/F_MED_Twin_Dark/Skins/Twin_Light_Purple/Materials/MF_TwinlightPurple
graph {nodes [
  n1 = TextureObjectParameter() {
   group=TwinlightPurple
   parameterName=Pattern Texture
   samplerType=SAMPLERTYPE_LinearColor
   texture=/BRCosmetics/Characters/Player/Female/Medium/Bodies/F_MED_Twin_Dark/Skins/Twin_Light_Purple/Textures/T_MED_StreetDemonOrange_Body_Flames
  }
  n2 = VertexInterpolator(VS=PreSkinnedPosition())
  n3 = VectorParameter() {group=TwinlightPurple, parameterName=TwinLightPurple_direction, defaultValue=<rgba -1 -1.6 3 1>}
  n4 = Multiply(A=Time(), B=@n3.RGB)
  n5 = MaterialFunctionCall(
   TextureObject=@n1
   TextureSize=Multiply(
    A=Constant3Vector() {constant=<rgba 200 200 200 0>}
    B=ScalarParameter() {defaultValue=0.8, group=TwinlightPurple, parameterName=TwinLightPurple_patternTile}
   )
   WorldPosition=Multiply(A=Subtract(A=@n2.PS, B=@n4), B=ScalarParameter() {defaultValue=3, group=TwinlightPurple, parameterName=TwinLightPurple_patternSpeed})
  ) {materialFunction=/Game/Weapons/WeaponSkins/Wraps/MaterialFunctions/WorldAlignedTexture}
  n6 = ComponentMask(None=@n5.XYZ Texture) {g=true}
  n7 = Add(A=ComponentMask(None=@n5.XYZ Texture) {r=true}, B=@n6)
  n8 = CurveAtlasRowParameter(CurveTime=Multiply(A=Frac(None=@n7), B=ScalarParameter() {defaultValue=0.95, group=TwinlightPurple, parameterName=Goo Colors blend bias})) {
   atlas=&/Twinlight_CurveAtlas
   curve=&/Twin_light_purple
   group=TwinlightPurple
   parameterName=MG_GooColor
   sortPriority=0
  }
  n9 = Fresnel(ExponentIn=ScalarParameter() {defaultValue=0.5, group=TwinlightPurple, parameterName=BrightnessFresnel_Exponentln})
  n10 = NamedRerouteDeclaration(None=LinearInterpolate(
    A=ScalarParameter() {defaultValue=0.2, group=TwinlightPurple, parameterName=swirl outline brightness}
    B=ScalarParameter() {defaultValue=0.5, group=TwinlightPurple, parameterName=swirl innter brightness}
    Alpha=OneMinus(None=@n9)
   )) {name=RedColorRIM, nodeColor=<rgba 1 0.4922 0 1>}
  n11 = ScalarParameter() {defaultValue=1, group=TwinlightPurple, parameterName=RedFresnel_Exponentln}
  n12 = Multiply(A=@n6, B=Power(Base=Fresnel(ExponentIn=@n11)))
  FunctionOutput(None=LinearInterpolate(A=Multiply(A=@n8, B=@n10), B=Constant3Vector() {constant=<rgba 20 0 0 0>}, Alpha=Power(Base=@n12)))
  n13 = VertexInterpolator(VS=PreSkinnedNormal())
  n14 = Fresnel(ExponentIn=Constant() {r=0.6}, Normal=@n13.PS)
  n15 = LinearInterpolate(
   A=ScalarParameter() {defaultValue=2, group=TwinlightPurple, parameterName=Seam_size}
   B=Constant() {r=1.5}
   Alpha=Saturate(None=Power(Base=@n14))
  ) {constA=1}
  NamedRerouteDeclaration(None=Saturate(None=Divide(A=@n13.PS, B=@n15))) {name=WorldspaceNormal, nodeColor=<rgba 1 0.9609 0 1>}
  n16 = PreSkinnedPosition()
  n17 = Multiply(A=Time(), B=@n3.RGB)
  n18 = ScalarParameter() {defaultValue=4, group=TwinlightPurple, parameterName=TwinLightPurple_noiseSpeed}
  n19 = MaterialFunctionCall(
   TextureObject=TextureObject() {samplerType=SAMPLERTYPE_LinearColor, texture=/Game/Effects/Fort_Effects/Textures/Plasma/T_Turbulence_AdditiveWarp_Linear}
   TextureSize=Multiply(
    A=Constant3Vector() {constant=<rgba 200 200 200 1>}
    B=ScalarParameter() {defaultValue=1.5, group=TwinlightPurple, parameterName=TwinLightPurple_NoiseTile}
   )
   WorldPosition=Subtract(A=VertexInterpolator(VS=@n16).PS, B=Multiply(A=@n17, B=@n18))
   World Space Normal=NamedRerouteUsage()
  ) {materialFunction=&WorldAlignedTexture}
  n20 = ScalarParameter() {defaultValue=50, group=TwinlightPurple, parameterName=TwinLightPurple_NoiseBrightness}
  n21 = PreSkinnedPosition()
  NamedRerouteDeclaration(None=Add(A=Multiply(A=@n19.XYZ Texture, B=@n20), B=VertexInterpolator(VS=@n21).PS)) {name=swirlNoise, nodeColor=<rgba 0 1 0.9531 1>}
  NamedRerouteUsage()
  NamedRerouteUsage()
 ]}
