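# Material function for a simplified "slurp" look on the Twin_Light_Purple skin, outputting base color, FX mask, roughness, specular and emissive.
# An emissive weight fades base color to black while raising emissive built from AO-shaded base color, dimmed by brightness and scaled by time of day.
# S_map channels blend into roughness and specular.
# An animated Fresnel bubble layer feeds an Epic-only QualitySwitch branch, but that switch reaches no output.
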MaterialFunction /BRCosmetics/Characters/Player/Female/Medium/Bodies/F_MED_Twin_Dark/Skins/Twin_Light_Purple/Materials/MF_simpleSlurp
graph {nodes [
  Constant3Vector()
  Constant() {r=1}
  Constant() {r=0.2}
  n1 = FunctionInput() {bUsePreviewValueAsDefault=true, inputName=baseColor}
  n2 = ScalarParameter() {defaultValue=1, group=Slurp, parameterName=Boots_emissiveWeight}
  FunctionOutput(None=LinearInterpolate(A=@n1, Alpha=@n2) {constB=0}) {outputName=Slurp_bascolor}
  n3 = ScalarParameter() {defaultValue=1, group=Slurp, parameterName=SimpleSlurpWeight}
  FunctionOutput(None=Multiply(A=FunctionInput() {bUsePreviewValueAsDefault=true, inputName=input_fxmask, inputType=FunctionInput_Scalar}, B=LinearInterpolate(Alpha=@n3))) {outputName=Slurp_FX mask}
  n4 = ScalarParameter() {defaultValue=0.2, group=Slurp, parameterName=Slurp_Roughness}
  n5 = FunctionInput() {bUsePreviewValueAsDefault=true, inputName=S_map}
  n6 = ComponentMask(None=@n5) {b=true}
  n7 = ScalarParameter() {defaultValue=0.5, group=Slurp, parameterName=Slurp_RoughnessMap_Intensity}
  FunctionOutput(None=NamedRerouteDeclaration(None=LinearInterpolate(A=@n4, B=@n6, Alpha=@n7)) {name=Roughness, nodeColor=<rgba 0 1 0.7188 1>}) {outputName=SlurpRougness}
  n8 = ScalarParameter() {defaultValue=0.5, group=Slurp, parameterName=Slurp_Specular}
  n9 = ScalarParameter() {defaultValue=1, group=Slurp, parameterName=Slurp_SpecularMap_Intensity}
  n10 = ComponentMask(None=@n5) {r=true}
  FunctionOutput(None=NamedRerouteDeclaration(None=LinearInterpolate(A=@n8, B=@n10, Alpha=@n9)) {name=SPC, nodeColor=<rgba 1 0 0.7266 1>}) {outputName=Slurp_Spc}
  n11 = Constant()
  n12 = ComponentMask(None=FunctionInput() {bUsePreviewValueAsDefault=true, inputName=AO}) {r=true}
  n13 = ScalarParameter() {defaultValue=1.3, group=Slurp, parameterName=Boots_AOintensity}
  n14 = LinearInterpolate(A=Multiply(A=Power(Base=@n12, Exp=@n13), B=@n1), Alpha=ScalarParameter() {defaultValue=0.3, group=Slurp, parameterName=Boots_brightness}) {constB=0}
  n15 = Multiply(
   A=MaterialFunctionCall(
    Value_Morning=ScalarParameter() {defaultValue=0.66, group=Slurp, parameterName=Slurp_TimeofDayScale_Morning}
    Value_Day=ScalarParameter() {defaultValue=1, group=Slurp, parameterName=Slurp_TimeofDayScale_Day}
    Value_Evening=ScalarParameter() {defaultValue=0.66, group=Slurp, parameterName=Slurp_TimeofDayScale_Evening}
    Value_Night=ScalarParameter() {defaultValue=0.5, group=Slurp, parameterName=Slurp_TimeofDayScale_Night}
   ) {materialFunction=&Scalar_Time_of_Day}.Result
   B=@n14
  )
  FunctionOutput(None=NamedRerouteDeclaration(None=LinearInterpolate(A=@n11, B=@n15, Alpha=@n2)) {name=Emissive, nodeColor=<rgba 0 1 0.1797 1>}) {outputName=Slurp_emiss}
  LinearInterpolate() {constB=0}
  Multiply(A=Constant3Vector() {constant=<rgba 200 200 200 1>}, B=Constant() {r=0.04})
  n16 = Multiply(
   A=MaterialFunctionCall!(ReflectionVector, ReflectionTexture) {materialFunction=/Engine/Functions/Engine_MaterialFunctions01/Reflections/ViewAlignedReflection}.Texture
   B=ScalarParameter() {defaultValue=1, group=Slurp, parameterName=Bubble_ViewDependantDistortion}
  )
  n17 = Time()
  n18 = AppendVector(
   A=ScalarParameter() {defaultValue=0.05, group=Slurp, parameterName=Bubble_ViewDistortion_SpeedX}
   B=ScalarParameter() {defaultValue=0.05, group=Slurp, parameterName=Bubble_ViewDistortion_SpeedY}
  )
  n19 = TextureSample(UVs=Rotator(Coordinate=Panner(Coordinate=@n16, Time=@n17, Speed=@n18), Time=@n17) {speed=0.0001}) {
   constMipValue=0
   texture=/BRCosmetics/Characters/Player/Female/Medium/Bodies/F_MED_Twin_Dark/Skins/Twin_Light_Purple/Textures/T_twinlightPurpleBlurredNoise
  }
  n20 = Power(Base=@n19.R)
  n21 = TextureObject() {texture=/BRCosmetics/Characters/Player/Female/Medium/Bodies/F_MED_Twin_Dark/Skins/Twin_Light_Purple/Textures/T_twinlightBubble}
  n22 = Multiply(A=Constant3Vector() {constant=<rgba 200 200 200 1>}, B=Constant() {r=0.025})
  n23 = Time()
  n24 = Constant3Vector() {constant=<rgba 0.66 1 2 1>}
  n25 = Subtract(A=VertexInterpolator(VS=PreSkinnedPosition()).PS, B=Multiply(A=Multiply(A=@n23, B=@n24), B=Constant() {r=0.3}))
  n26 = Time()
  n27 = Subtract(A=VertexInterpolator(VS=PreSkinnedPosition()).PS, B=Multiply(A=Multiply(A=@n26, B=@n24), B=Constant() {r=0.5}))
  n28 = Max(
   A=ComponentMask(None=MaterialFunctionCall(TextureObject=@n21, TextureSize=@n22, WorldPosition=@n25) {materialFunction=/Game/Weapons/WeaponSkins/Wraps/MaterialFunctions/WorldAlignedTexture}.XYZ Texture) {r=true}
   B=ComponentMask(None=MaterialFunctionCall(TextureObject=@n21, TextureSize=@n22, WorldPosition=@n27) {materialFunction=&WorldAlignedTexture}.XYZ Texture) {g=true}
  )
  n29 = Constant() {r=0.32}
  n30 = Multiply(A=Desaturation(None=Multiply(A=@n28, B=@n29), Fraction=Constant() {r=1}), B=Constant3Vector() {constant=<rgba 0.52 0.2584 0.52 0>})
  n31 = Constant() {r=0.3}
  NamedRerouteDeclaration(None=LinearInterpolate(B=Multiply(A=@n20, B=@n30), Alpha=Fresnel(ExponentIn=@n31))) {name=simple bubble, nodeColor=<rgba 0 0.3906 1 1>}
  PreSkinnedPosition()
  QualitySwitch(Default=@n14, Low=@n14, High=@n14, Medium=@n14, Epic=Add(A=NamedRerouteUsage(), B=@n14))
  Reroute()
  VertexInterpolator()
  VertexInterpolator(VS=PreSkinnedNormal())
 ]}
