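# Master body material for the Twin Light Purple female character skin, using subsurface profile and clear coat shading.
# PreFX output passes through optional gem, clear coat and fresnel-remap functions, plus an always-applied emissive fresnel.
# Split attributes get switchable purple swirl emissive on the arm mask and a "simpleSlurp" overlay on color, specular, roughness and emissive.
# SkinFX_Mask channels drive effect masks, and character and cinematic PostFX functions finish the result.
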Material /BRCosmetics/Characters/Player/Female/Medium/Bodies/F_MED_Twin_Dark/Skins/Twin_Light_Purple/Materials/F_MED_Twin_Light_Purple_Body_Master
graph {
 properties {
  bAutomaticallySetUsageInEditor=false
  bUseMaterialAttributes=true
  bUsedWithClothing=true
  bUsedWithMeshDeformer=true
  bUsedWithMorphTargets=true
  bUsedWithNanite=true
  bUsedWithSkeletalMesh=true
  pixelDepthOffsetMode=PDOM_Legacy
  shadingModel=MSM_FromMaterialExpression
  subsurfaceProfile=&SS_HeroSkin_02
  usedShadingModels=Default Lit | Clear Coat | Subsurface Profile
 }
 nodes [
  n1 = MaterialFunctionCall() {materialFunction=&BaseFN_CharacterFunction_PreFX}
  n2 = NamedRerouteUsage()
  n3 = MaterialFunctionCall!(baseColor=NamedRerouteUsage(), S_map=NamedRerouteUsage(), input_fxmask=NamedRerouteUsage(), AO=NamedRerouteUsage()) {materialFunction=&/MF_simpleSlurp}
  n4 = NamedRerouteUsage()
  n5 = NamedRerouteUsage()
  n6 = TextureSample(
   UVs=TextureCoordinate()
   Tex=NamedRerouteDeclaration(None=TextureObjectParameter() {group=Skin FX, parameterName=SkinFX_Mask}) {name=SkinFX_Mask, nodeColor=<rgba 0.9842 0 1 1>}
  )
  n7 = Power(Base=StaticComponentMaskParameter(None=@n6.RGBA) {defaultB=true, group=Skin FX, parameterName=ArmMask}, Exp=Constant() {r=1.8})
  n8 = MaterialFunctionCall() {materialFunction=&/MF_TwinlightPurple}
  n9 = StaticSwitchParameter(True=LinearInterpolate(A=NamedRerouteUsage(), B=@n8.Result, Alpha=Saturate(None=@n7)), False=NamedRerouteUsage()) {defaultValue=true, group=Skin FX, parameterName=use swirl pattern body}
  n10 = SetMaterialAttributes!(
   MaterialAttributes
   Base Color=StaticSwitchParameter(True=LinearInterpolate(A=@n2, B=@n3.Slurp_bascolor, Alpha=@n3.Slurp_bascolor) {constB=0}, False=@n2) {defaultValue=true, group=Skin FX, parameterName=useSimpleSlurp}
   Specular=StaticSwitchParameter(True=LinearInterpolate(A=@n4, B=@n3.Slurp_Spc, Alpha=@n3.Slurp_Spc), False=@n4) {defaultValue=true, group=Skin FX, parameterName=useSimpleSlurp}
   Roughness=StaticSwitchParameter(True=LinearInterpolate(A=@n5, B=@n3.SlurpRougness, Alpha=@n3.SlurpRougness), False=@n5) {defaultValue=true, group=Skin FX, parameterName=useSimpleSlurp}
   Emissive Color=StaticSwitchParameter(True=LinearInterpolate(A=@n9, B=@n3.Slurp_emiss, Alpha=@n3.Slurp_emiss), False=@n9) {defaultValue=true, group=Skin FX, parameterName=useSimpleSlurp}
   Normal=NamedRerouteUsage()
   Metallic=NamedRerouteUsage()
   Clear Coat or Curvature or Custom Data 0=NamedRerouteUsage()
   Clear Coat Roughness or Custom Data 1=NamedRerouteUsage()
  ) {attributeSetTypes [
    #1
    #2
    #3
    #4
    #5
    #6
    #7
    #8
   ]}
  n11 = MaterialFunctionCall(AO=@n1.AO, Distance Blend=@n1.AO, Fresnel=@n1.AO, Gmap=@n1.AO, Pre Skinned Local Position=@n1.AO, Material=@n10) {materialFunction=&BaseFN_CharacterFunction_PostFX}
  MaterialFunctionCall(In=@n11.Result) {materialFunction=&CINE_MF_CharPostFX}
  n12 = StaticComponentMaskParameter(None=@n6.RGBA) {defaultR=true, group=Skin FX, parameterName=Generic Gem Effect Channel}
  n13 = StaticComponentMaskParameter(None=@n6.RGBA) {defaultB=true, group=Skin FX, parameterName=Generic Gem Color Change Channel}
  n14 = MaterialFunctionCall!(In=StaticSwitchParameter(
    True=MaterialFunctionCall!(MASK=@n12, Color Change Sequence=@n13, Material=@n1.Result) {materialFunction=/Game/Weapons/WeaponSkins/MaterialFunction/GenericGem}.Result
    False=@n1.Result
   ) {defaultValue=true, group=Skin FX, parameterName=UseGenericGem}) {materialFunction=&/MF_EmissiveFresnel_twinlight}
  n15 = StaticSwitchParameter(
   True=MaterialFunctionCall!(Alpha=StaticComponentMaskParameter(None=@n6.RGBA) {defaultR=true, group=Skin FX, parameterName=Clear Coat Channel}, In=@n14.Result) {materialFunction=/Game/Weapons/WeaponSkins/MaterialFunction/Clearcoat}.Result
   False=@n14.Result
  ) {group=Skin FX, parameterName=Use Clear Coat}
  n16 = NamedRerouteUsage()
  n17 = GetMaterialAttributes(None=StaticSwitchParameter(
    True=MaterialFunctionCall!(MA=@n15, FXMask=@n16, PreskinLocalPosition=@n1.Pre skinned Local Position) {materialFunction=/Game/Weapons/WeaponSkins/MaterialFunction/RemapCurveByFresnel}.Result
    False=@n15
   ) {group=Skin FX, parameterName=UseRemapCurveByFresnel}) {attributeGetTypes [
    #1
    #6
    #2
    #3
    #4
    #5
    #7
    #8
   ]}
  NamedRerouteDeclaration(None=@n17.Clear Coat or Curvature or Custom Data 0) {name=clearcoat, nodeColor=<rgba 1 0.0938 0 1>}
  NamedRerouteDeclaration(None=StaticComponentMaskParameter(None=@n6.RGBA) {defaultR=true, group=Skin FX, parameterName=Bootsmask}) {name=boots mask, nodeColor=<rgba 0.2891 1 0 1>}
  NamedRerouteDeclaration(None=@n1.AO) {name=Ao, nodeColor=<rgba 0.5 0 1 1>}
  NamedRerouteDeclaration(None=@n17.Clear Coat Roughness or Custom Data 1) {name=Clear coatroughness, nodeColor=<rgba 0.9219 0 1 1>}
  NamedRerouteDeclaration(None=@n17.Normal) {name=normal, nodeColor=<rgba 0 1 0.0156 1>}
  NamedRerouteDeclaration(None=@n17.Base Color) {name=Base Color, nodeColor=<rgba 0 0.8125 1 1>}
  NamedRerouteDeclaration(None=@n17.Metallic) {name=Metallic, nodeColor=<rgba 0 0.8125 1 1>}
  NamedRerouteDeclaration(None=@n17.Specular) {name=Spec, nodeColor=<rgba 0 0.8125 1 1>}
  NamedRerouteDeclaration(None=@n17.Roughness) {name=Roughness, nodeColor=<rgba 0 0.8125 1 1>}
  NamedRerouteDeclaration(None=@n17.Emissive Color) {name=Emissive, nodeColor=<rgba 0 0.8125 1 1>}
  NamedRerouteUsage()
  NamedRerouteUsage()
 ]
}
