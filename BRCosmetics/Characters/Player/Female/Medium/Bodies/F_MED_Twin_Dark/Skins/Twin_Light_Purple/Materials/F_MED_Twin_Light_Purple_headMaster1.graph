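# Head master material for the Twin Light Purple skin of a medium female player character, using material attributes and a per-expression shading model.
# BaseFN_CharacterFunction_PreFX output runs through a twinlight emissive fresnel function, then gets split into attributes.
# A SkinFX_Mask channel (Neckmask, blue by default) blends MF_TwinlightPurple into Base Color and Emissive.
# The result passes through character and cinematic PostFX functions; a CharacterSkinFX call feeds nothing downstream.
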Material /BRCosmetics/Characters/Player/Female/Medium/Bodies/F_MED_Twin_Dark/Skins/Twin_Light_Purple/Materials/F_MED_Twin_Light_Purple_headMaster1
graph {
 properties {
  bAutomaticallySetUsageInEditor=false
  bUseMaterialAttributes=true
  bUsedWithClothing=true
  bUsedWithMeshDeformer=true
  bUsedWithMorphTargets=true
  bUsedWithNanite=true
  bUsedWithSkeletalMesh=true
  pixelDepthOffsetMode=PDOM_Legacy
  refractionMethod=RM_IndexOfRefraction
  shadingModel=MSM_FromMaterialExpression
  subsurfaceProfile=&SS_HeroSkin_02
  usedShadingModels=Default Lit | Subsurface | Clear Coat | Subsurface Profile | Cloth | Eye
 }
 nodes [
  n1 = MaterialFunctionCall() {materialFunction=&BaseFN_CharacterFunction_PreFX}
  n2 = MaterialFunctionCall!(In=@n1.Result) {materialFunction=&/MF_EmissiveFresnel_twinlight}
  n3 = GetMaterialAttributes(None=@n2.Result) {attributeGetTypes [
    #1
    #2
    #3
    #4
    #5
    #6
   ]}
  n4 = StaticComponentMaskParameter(None=TextureSample(UVs=TextureCoordinate(), Tex=TextureObjectParameter() {group=Skin FX, parameterName=SkinFX_Mask}).RGBA) {defaultB=true, group=Skin FX, parameterName=Neckmask}
  n5 = MaterialFunctionCall() {materialFunction=&/MF_TwinlightPurple}
  n6 = SetMaterialAttributes!(
   MaterialAttributes=@n3.MaterialAttributes
   Base Color=LinearInterpolate(A=NamedRerouteUsage(), B=@n5.Result, Alpha=@n4)
   Curvature or Metallic=NamedRerouteUsage()
   Specular=NamedRerouteUsage()
   Roughness=NamedRerouteUsage()
   Emissive Color=LinearInterpolate(A=NamedRerouteUsage(), B=@n5.Result, Alpha=@n4)
   Normal=NamedRerouteUsage()
  ) {attributeSetTypes [
    #1
    #2
    #3
    #4
    #5
    #6
   ]}
  n7 = MaterialFunctionCall(AO=@n1.AO, Distance Blend=@n1.AO, Fresnel=@n1.AO, Gmap=@n1.AO, Pre Skinned Local Position=@n1.AO, Material=@n6) {materialFunction=&BaseFN_CharacterFunction_PostFX}
  MaterialFunctionCall(In=@n7.Result) {materialFunction=&CINE_MF_CharPostFX}
  MaterialFunctionCall(
   PreskinnedPosition=@n1.Pre skinned Local Position
   Gmap=@n1.Pre skinned Local Position
   MaterialAttributes=@n1.Pre skinned Local Position
   AO=@n1.Pre skinned Local Position
  ) {materialFunction=&CharacterSkinFX}
  NamedRerouteDeclaration(None=@n3.Normal) {name=normal, nodeColor=<rgba 0 1 0.0156 1>}
  NamedRerouteDeclaration(None=@n3.Base Color) {name=Base Color, nodeColor=<rgba 0 0.8125 1 1>}
  NamedRerouteDeclaration(None=@n3.Curvature or Metallic) {name=Metallic, nodeColor=<rgba 0 0.8125 1 1>}
  NamedRerouteDeclaration(None=@n3.Specular) {name=Spec, nodeColor=<rgba 0 0.8125 1 1>}
  NamedRerouteDeclaration(None=@n3.Roughness) {name=Roughness, nodeColor=<rgba 0 0.8125 1 1>}
  NamedRerouteDeclaration(None=@n3.Emissive Color) {name=Emissive, nodeColor=<rgba 0 0.8125 1 1>}
 ]
}
